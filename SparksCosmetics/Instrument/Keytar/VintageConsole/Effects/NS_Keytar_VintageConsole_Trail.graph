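# Niagara trail effect for the Vintage Console Keytar cosmetic instrument, drawing animation-driven ribbons between two sockets.
# Two emitters, Keys and Soft, spawn and update anim-trail particles through a shared system-level Socket Reader.
# Both fade over short lifetimes and adjust tint and opacity by time of day.
# Keys also writes dynamic material parameters and switches its starting color between gameplay and front end, while Soft uses a fixed warm tint.
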NiagaraSystem /SparksCosmetics/Instrument/Keytar/VintageConsole/Effects/NS_Keytar_VintageConsole_Trail
dependencies {
 usedDataInterfaces [
  &NiagaraDataInterfaceCurve
  &NiagaraDataInterfaceColorCurve
  &NiagaraDataInterfaceSkeletalMesh
  &NiagaraDataInterfaceSocketReader
 ]
 usedDynamicInputs [&LerpLinearColors]
 usedModules [
  &self:SystemSpawnScript.SystemScriptSource.SystemScriptGraph.NiagaraNodeAssignment_2.SetVariables_#1
  &SystemState
  &self:SystemSpawnScript.SystemScriptSource.SystemScriptGraph.NiagaraNodeAssignment_0.SetVariables_#2
  &self:SystemSpawnScript.SystemScriptSource.SystemScriptGraph.NiagaraNodeAssignment_1.SetVariables_#3
 ]
}
emitters {
 FN_AnimTrail_Keys {
  emitterSpawnScript [&FN_RibbonIDOffsetUtility()]
  emitterUpdateScript [
   &EmitterState()
   &FN_SpawnAnimTrail()
  ]
  particleSpawnScript [
   &FN_InitializeAnimTrailBoneIndecies(First Bone or Socket Index=1, Second Bone or Socket Index=0)
   /Game/Effects/Niagara/Modules/AnimTrail/FN_InitializeAnimTrail/FN_InitializeAnimTrail_SocketReader/FN_InitializeAnimTrail_SocketReader(
    Color:LinearColor=&MergeHelper_ColorSwitch(
     0=<rgba 1 1 1 0.7>
     1=<rgba 1 1 1 0.1>
     2=?
     3=?
     Index=&MakeIntFromFloat(Float=&FN_FrontEndFloatModifier(Gameplay=0, Front End=1))
     Num=::NewEnumerator0
    )
    Lifetime=0.15
    Socket Reader:NiagaraDataInterfaceSocketReader=$System.Socket Reader
   )
   &AddVelocity(Velocity=<800 0 0>) {enabled=false}
   &DynamicMaterialParameters(
    Index 0 Param 1=1
    Index 0 Param 4=1
    Module.Index0Param0Write=true
    Module.Index0Param1Write=true
    Module.Index0Param2Write=true
    Module.Index0Param3Write=true
   )
  ]
  particleUpdateScript [
   &SolveForcesAndVelocity(Clamp Velocity=false, Limit Acceleration=false)
   /Game/Effects/Niagara/Modules/AnimTrail/FN_UpdateAnimTrail/FN_UpdateAnimTrail_SocketReader/FN_UpdateAnimTrail_SocketReader(Socket Reader:NiagaraDataInterfaceSocketReader=$System.Socket Reader)
   &ScaleColor(
    Scale Alpha=&Multiply_Float(
     A=&FloatFromCurve(FloatCurve=DataInterface() {propertyValues {Curve {keys [
          {interpMode=RCIM_Cubic, leaveTangent=-1.7891, tangentMode=RCTM_Break, time=0, value=1}
          <at 0.5 0.05>
          {arriveTangent=-0.2139, interpMode=RCIM_Cubic, tangentMode=RCTM_Break, time=1, value=0}
         ]}}})
     B=&FNTimeOfDayScalar(Night Scale=0.65, Morning Scale=0.3, Day Scale=0.7, Evening Scale=0.7)
    )
    Scale RGB=&MakeVectorFromLinearColorRGB(Color=&ColorFromCurve(
      ColorCurve=DataInterface() {propertyValues {
        AlphaCurve {keys [<at 0 1 RCIM_Cubic>]}
        BlueCurve {keys [<at 0 1 RCIM_Cubic>,<at 0.25 0.9061>,<at 0.5 1>,<at 0.75 1>,<at 1 1 RCIM_Cubic>]}
        GreenCurve {keys [<at 0 0.25 RCIM_Cubic>,<at 0.25 0.9061>,<at 0.5 1>,<at 0.75 0.8365>,<at 1 0.25 RCIM_Cubic>]}
        RedCurve {keys [<at 0 0.2609 RCIM_Cubic>,<at 0.25 1>,<at 0.5 1>,<at 0.75 0.9513>,<at 1 0.2609 RCIM_Cubic>]}
       }}
      CurveIndex=Linked() {linkedVariable {name=NPC.FortniteNPC.TimeOfDay_0-1, type=NiagaraFloat}}
     ))
    ScaleA=true
    ScaleRGB=true
   )
  ]
  rendererClasses [&NiagaraRibbonRendererProperties]
  renderers [{rendererClass=&NiagaraRibbonRendererProperties}]
 }
 FN_AnimTrail_Soft {
  emitterSpawnScript [&FN_RibbonIDOffsetUtility()]
  emitterUpdateScript [
   &EmitterState()
   &FN_SpawnAnimTrail()
  ]
  particleSpawnScript [
   &FN_InitializeAnimTrailBoneIndecies(First Bone or Socket Index=1, Second Bone or Socket Index=0)
   &FN_InitializeAnimTrail_SocketReader(Color:LinearColor=<rgba 0.1774 0.127 0.104 1>, Lifetime=0.2, Socket Reader:NiagaraDataInterfaceSocketReader=$System.Socket Reader)
   &AddVelocity(Velocity=<800 0 0>) {enabled=false}
  ]
  particleUpdateScript [
   &SolveForcesAndVelocity(Clamp Velocity=false, Limit Acceleration=false)
   &FN_UpdateAnimTrail_SocketReader(Socket Reader:NiagaraDataInterfaceSocketReader=$System.Socket Reader)
   &ScaleColor(
    Scale Alpha=&Multiply_Float(
     A=&FloatFromCurve(FloatCurve=DataInterface() {propertyValues {Curve {keys [
          {interpMode=RCIM_Cubic, leaveTangent=-0.5911, tangentMode=RCTM_Break, time=0, value=1}
          <at 0.6689 0.0233>
          {arriveTangent=-0.2139, interpMode=RCIM_Cubic, tangentMode=RCTM_Break, time=1, value=0}
         ]}}})
     B=&FNTimeOfDayScalar(Night Scale=1, Morning Scale=1, Evening Scale=1)
    )
    Scale RGB=&FN_Select_Vector_from_Time_Of_Day(Morning=<1 1 1>, Day=<1 1 1>, Evening=<1 1 1>, Night=<0.8 0.8 1>)
    ScaleA=true
    ScaleRGB=true
   ) {module=ScaleColor002}
  ]
  rendererClasses [&NiagaraRibbonRendererProperties]
  renderers [{rendererClass=&NiagaraRibbonRendererProperties}]
 }
}
summary {
 systemName=NS_Keytar_VintageConsole_Trail
 emitters [
  {emitterName=FN_AnimTrail_Soft, rendererClasses [&NiagaraRibbonRendererProperties]}
  {emitterName=FN_AnimTrail_Keys, rendererClasses [&NiagaraRibbonRendererProperties]}
 ]
 userVariables [
  User.bIsFrontEnd {type=NiagaraBool}
  User.bIsFrontEndPreview {type=NiagaraBool}
  User.RibbonIDOffset {type=NiagaraInt32}
 ]
}
