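# Physics asset for the Vintage Console keytar, simulating its upper and lower dynamic cords.
# Each cord has top and bottom three-capsule chains rooted on the handle capsule.
# Root and middle joints lock twist but leave swing unlocked, while end joints lock both.
# Locked constraints also join each cord's top and bottom end segments, closing it into a loop through the handle.
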PhysicsAsset /SparksCosmetics/Instrument/Keytar/VintageConsole/Meshes/VintageConsole_Keytar_Physics
bodies {
 handle [{shapeType=Capsule, length=22.8512, radius=4, center=<0 -2.273 0>}]
 lower_dyn_cord_bottom_01 [{
   shapeType=Capsule
   length=1.102
   radius=1.3831
   shapeName=lower_dyn_cord_bottom_01_capsule
   center=<0.7473 0.0568 0.1011>
   rotation=<rot -2.4379 -88.446 -78.2467>
  }]
 lower_dyn_cord_bottom_02 [{
   shapeType=Capsule
   length=1.4691
   radius=1.3242
   shapeName=lower_dyn_cord_bottom_02_capsule
   center=<0.7756 0.1291 0.0391>
   rotation=<rot 70 -0.0005 5.7014>
  }]
 lower_dyn_cord_bottom_root [{
   shapeType=Capsule
   length=1.8423
   radius=1.3703
   shapeName=lower_dyn_cord_bottom_root_capsule
   center=<2.7909 -0.0052 -0.0021>
   rotation=<rot 0 -90 -78>
  }]
 lower_dyn_cord_top_01 [{
   shapeType=Capsule
   length=1.1116
   radius=1.3123
   shapeName=lower_dyn_cord_top_01_capsule
   center=<0.8245 -0.055 0.3509>
   rotation=<rot 12.6239 95.2591 -68.7546>
  }]
 lower_dyn_cord_top_02 [{
   shapeType=Capsule
   length=1.4129
   radius=1.2964
   shapeName=lower_dyn_cord_top_02_capsule
   center=<0.6312 0.1334 -0.117>
   rotation=<rot 73.0215 0.0011 0>
  }]
 lower_dyn_cord_top_root [{
   shapeType=Capsule
   length=1.5448
   radius=1.4458
   shapeName=lower_dyn_cord_top_root_capsule
   center=<2.6414 -0.0748 0.6744>
   rotation=<rot 0.7979 89.355 -84.9123>
  }]
 upper_dyn_cord_bottom_01 [{
   shapeType=Capsule
   length=1.3517
   radius=1.2314
   shapeName=upper_dyn_cord_bottom_01_capsule
   center=<0.854 0.1371 0.0437>
   rotation=<rot 76.5417 -0.0001 -1.8264>
  }]
 upper_dyn_cord_bottom_02 [{
   shapeType=Capsule
   length=1.5738
   radius=1.3452
   shapeName=upper_dyn_cord_bottom_02_capsule
   center=<0.6401 0.2847 -0.2068>
   rotation=<rot 0.3338 92.5896 -84.0778>
  }]
 upper_dyn_cord_bottom_root [{
   shapeType=Capsule
   length=1.4364
   radius=1.3539
   shapeName=upper_dyn_cord_bottom_root_capsule
   center=<2.4149 -0.2196 -0.2516>
   rotation=<rot -85.9714 179.9995 -177.2341>
  }]
 upper_dyn_cord_top_01 [{
   shapeType=Capsule
   length=2.5689
   radius=1.4498
   shapeName=upper_dyn_cord_top_01_capsule
   center=<0.2924 0.1289 -0.222>
   rotation=<rot -3.8985 93.9253 -73.9262>
  }]
 upper_dyn_cord_top_02 [{
   shapeType=Capsule
   length=1.6166
   radius=1.3122
   shapeName=upper_dyn_cord_top_02_capsule
   center=<0.3324 0.259 0.1788>
   rotation=<rot 10.7018 -79.6631 -70.8711>
  }]
 upper_dyn_cord_top_root [{
   shapeType=Capsule
   length=2.4455
   radius=1.3454
   shapeName=upper_dyn_cord_top_root_capsule
   center=<2.7455 0.0195 0.367>
   rotation=<rot -3.6812 88.7738 -80.5448>
  }]
}
constraints [
 {
  bone1Name=lower_dyn_cord_bottom_root
  bone2Name=handle
  swing1LimitDegrees=20
  swing2LimitDegrees=20
  twistLimitDegrees=45
  twistMotion=Locked
 }
 {
  bone1Name=lower_dyn_cord_bottom_01
  bone2Name=lower_dyn_cord_bottom_root
  swing1LimitDegrees=20
  swing2LimitDegrees=20
  twistLimitDegrees=45
  twistMotion=Locked
 }
 {
  bone1Name=lower_dyn_cord_bottom_02
  bone2Name=lower_dyn_cord_bottom_01
  swing1LimitDegrees=2
  swing1Motion=Locked
  swing2LimitDegrees=2
  swing2Motion=Locked
  twistLimitDegrees=45
  twistMotion=Locked
 }
 {bone1Name=lower_dyn_cord_top_root, bone2Name=handle, swing1LimitDegrees=35, swing2LimitDegrees=20, twistLimitDegrees=45, twistMotion=Locked}
 {
  bone1Name=lower_dyn_cord_top_01
  bone2Name=lower_dyn_cord_top_root
  swing1LimitDegrees=20
  swing2LimitDegrees=20
  twistLimitDegrees=45
  twistMotion=Locked
 }
 {
  bone1Name=lower_dyn_cord_top_02
  bone2Name=lower_dyn_cord_top_01
  swing1LimitDegrees=2
  swing1Motion=Locked
  swing2LimitDegrees=2
  swing2Motion=Locked
  twistLimitDegrees=45
  twistMotion=Locked
 }
 {
  bone1Name=upper_dyn_cord_bottom_root
  bone2Name=handle
  swing1LimitDegrees=20
  swing2LimitDegrees=30
  twistLimitDegrees=45
  twistMotion=Locked
 }
 {
  bone1Name=upper_dyn_cord_bottom_01
  bone2Name=upper_dyn_cord_bottom_root
  swing1LimitDegrees=20
  swing2LimitDegrees=20
  twistLimitDegrees=45
  twistMotion=Locked
 }
 {bone1Name=upper_dyn_cord_top_root, bone2Name=handle, swing1LimitDegrees=30, swing2LimitDegrees=15, twistLimitDegrees=45, twistMotion=Locked}
 {
  bone1Name=upper_dyn_cord_top_01
  bone2Name=upper_dyn_cord_top_root
  swing1LimitDegrees=20
  swing2LimitDegrees=20
  twistLimitDegrees=45
  twistMotion=Locked
 }
 {
  bone1Name=upper_dyn_cord_top_02
  bone2Name=upper_dyn_cord_top_01
  swing1LimitDegrees=45
  swing1Motion=Locked
  swing2LimitDegrees=45
  swing2Motion=Locked
  twistLimitDegrees=45
  twistMotion=Locked
 }
 {
  bone1Name=upper_dyn_cord_bottom_02
  bone2Name=upper_dyn_cord_bottom_01
  swing1LimitDegrees=45
  swing1Motion=Locked
  swing2LimitDegrees=45
  swing2Motion=Locked
  twistLimitDegrees=45
  twistMotion=Locked
 }
 {
  bone1Name=upper_dyn_cord_top_02
  bone2Name=upper_dyn_cord_bottom_02
  swing1LimitDegrees=45
  swing1Motion=Locked
  swing2LimitDegrees=45
  swing2Motion=Locked
  twistLimitDegrees=45
  twistMotion=Locked
 }
 {
  bone1Name=lower_dyn_cord_top_02
  bone2Name=lower_dyn_cord_bottom_02
  swing1LimitDegrees=45
  swing1Motion=Locked
  swing2LimitDegrees=45
  swing2Motion=Locked
  twistLimitDegrees=45
  twistMotion=Locked
 }
]
